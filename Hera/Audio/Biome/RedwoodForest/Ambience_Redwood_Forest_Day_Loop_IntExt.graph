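# Looping daytime ambience SoundCue for the Redwood Forest biome, routed to the indoor ambient sound class with ambient attenuation.
# A quality-level switch plays a quad chain in its first two slots and a stereo chain in the last two.
# Each mixes forest creak and calm grassland loops, pitching up creak in quad but grassland in stereo.
# Nested looping envelopers add slow pitch and volume drift on differing periods.
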SoundCue /Hera/Audio/Biome/RedwoodForest/Ambience_Redwood_Forest_Day_Loop_IntExt
root=QualityLevel_0
nodes [
 {
  id=Enveloper_4
  children [Mixer_1]
  properties {
   bLoop=true
   bLoopIndefinitely=true
   loopEnd=30
   pitchCurve {editorCurveData {keys [
      {arriveTangent=-0.0013, interpMode=RCIM_Cubic, leaveTangent=-0.0013, time=0}
      <at 15 0.98 RCIM_Cubic>
      {arriveTangent=0.0013, interpMode=RCIM_Cubic, leaveTangent=0.0013, time=30}
     ]}}
   volumeCurve {editorCurveData {keys [<at 0 1>]}}
  }
 }
 {
  id=Enveloper_5
  children [Enveloper_4]
  properties {
   bLoop=true
   bLoopIndefinitely=true
   loopEnd=20
   pitchCurve {editorCurveData {keys [{time=0}]}}
   volumeCurve {editorCurveData {keys [
      {arriveTangent=-0.01, interpMode=RCIM_Cubic, time=0, value=1}
      <at 10 1.1 RCIM_Cubic>
      {interpMode=RCIM_Cubic, leaveTangent=0.01, time=20, value=1}
     ]}}
  }
 }
 {
  id=Enveloper_6
  children [Mixer_0]
  properties {
   bLoop=true
   bLoopIndefinitely=true
   loopEnd=30
   pitchCurve {editorCurveData {keys [
      {arriveTangent=-0.0013, interpMode=RCIM_Cubic, leaveTangent=-0.0013, time=0}
      <at 15 0.98 RCIM_Cubic>
      {arriveTangent=0.0013, interpMode=RCIM_Cubic, leaveTangent=0.0013, time=30}
     ]}}
   volumeCurve {editorCurveData {keys [<at 0 1>]}}
  }
 }
 {
  id=Enveloper_7
  children [Enveloper_6]
  properties {
   bLoop=true
   bLoopIndefinitely=true
   loopEnd=20
   pitchCurve {editorCurveData {keys [{time=0}]}}
   volumeCurve {editorCurveData {keys [
      {arriveTangent=-0.01, interpMode=RCIM_Cubic, time=0, value=1}
      <at 10 1.1 RCIM_Cubic>
      {interpMode=RCIM_Cubic, leaveTangent=0.01, time=20, value=1}
     ]}}
  }
 }
 {id=Mixer_0, children [Modulator_0,WavePlayer_4], properties {inputVolume [1,0.8]}}
 {id=Mixer_1, children [Modulator_1,WavePlayer_1], properties {inputVolume [1,0.8]}}
 {id=Modulator_0, children [WavePlayer_0], properties {pitchMax=1.1, pitchMin=1.1}}
 {id=Modulator_1, children [WavePlayer_5], properties {pitchMax=1.1, pitchMin=1.1}}
 {id=QualityLevel_0, children [Enveloper_7,Enveloper_7,Enveloper_5,Enveloper_5]}
 {id=WavePlayer_0, properties {bLooping=true, soundWaveAssetPtr=/AmbientAudioContent/Waves/Biomes/Ambience_Forest_Creak_Day_Loop_Quad}}
 {id=WavePlayer_1, properties {bLooping=true, soundWaveAssetPtr=/AmbientAudioContent/Waves/Biomes/Ambience_Forest_Creak_Day_Loop_Stereo}}
 {id=WavePlayer_4, properties {bLooping=true, soundWaveAssetPtr=/AmbientAudioContent/Waves/Biomes/Ambience_Hera_Grasslands_Calm_Day_Loop_Quad}}
 {
  id=WavePlayer_5
  properties {bLooping=true, soundWaveAssetPtr=/AmbientAudioContent/Waves/Biomes/Ambience_Hera_Grasslands_Calm_Day_Loop_Stereo}
 }
]
properties {
 attenuationSettings=/AmbientAudioBase/Attenuations/Fort_Ambient_Athena_Att
 maxDistance=2097152
 pitchMultiplier=0.9
 soundClassObject=&Fort_Ambient_Indoors_Athena
 volumeMultiplier=0.6
 preEffectBusSends [{}]
}
