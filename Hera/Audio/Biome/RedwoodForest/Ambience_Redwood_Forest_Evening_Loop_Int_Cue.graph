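# Looping indoor evening ambience cue for the Redwood Forest biome, played quietly at lowered pitch through the Athena indoor ambient sound class.
# A quality node sends the first two slots to a quad chain and the last two to a stereo chain.
# Each chain mixes a prominent chaparral morning bed with softer crickets.
# Nested endlessly looping envelopers add slow pitch and volume modulation on different periods, varying the static loops.
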SoundCue /Hera/Audio/Biome/RedwoodForest/Ambience_Redwood_Forest_Evening_Loop_Int_Cue
root=QualityLevel_2
nodes [
 {
  id=Enveloper_0
  children [Mixer_1]
  properties {
   bLoop=true
   bLoopIndefinitely=true
   loopEnd=30
   pitchCurve {editorCurveData {keys [
      {arriveTangent=-0.0013, interpMode=RCIM_Cubic, leaveTangent=-0.0013, time=0}
      <at 15 0.98 RCIM_Cubic>
      {arriveTangent=0.0013, interpMode=RCIM_Cubic, leaveTangent=0.0013, time=30}
     ]}}
   volumeCurve {editorCurveData {keys [<at 0 1>]}}
  }
 }
 {
  id=Enveloper_1
  children [Mixer_0]
  properties {
   bLoop=true
   bLoopIndefinitely=true
   loopEnd=30
   pitchCurve {editorCurveData {keys [
      {arriveTangent=-0.0013, interpMode=RCIM_Cubic, leaveTangent=-0.0013, time=0}
      <at 15 0.98 RCIM_Cubic>
      {arriveTangent=0.0013, interpMode=RCIM_Cubic, leaveTangent=0.0013, time=30}
     ]}}
   volumeCurve {editorCurveData {keys [<at 0 1>]}}
  }
 }
 {
  id=Enveloper_2
  children [Enveloper_1]
  properties {
   bLoop=true
   bLoopIndefinitely=true
   loopEnd=20
   pitchCurve {editorCurveData {keys [{time=0}]}}
   volumeCurve {editorCurveData {keys [
      {arriveTangent=-0.01, interpMode=RCIM_Cubic, leaveTangent=-0.01, time=0, value=1}
      <at 10 0.9 RCIM_Cubic>
      {arriveTangent=0.01, interpMode=RCIM_Cubic, leaveTangent=0.01, time=20, value=1}
     ]}}
  }
 }
 {
  id=Enveloper_3
  children [Enveloper_0]
  properties {
   bLoop=true
   bLoopIndefinitely=true
   loopEnd=20
   pitchCurve {editorCurveData {keys [{time=0}]}}
   volumeCurve {editorCurveData {keys [
      {arriveTangent=-0.01, interpMode=RCIM_Cubic, leaveTangent=-0.01, time=0, value=1}
      <at 10 0.9 RCIM_Cubic>
      {arriveTangent=0.01, interpMode=RCIM_Cubic, leaveTangent=0.01, time=20, value=1}
     ]}}
  }
 }
 {id=Mixer_0, children [WavePlayer_5,WavePlayer_3], properties {inputVolume [1.5,0.6]}}
 {id=Mixer_1, children [WavePlayer_2,WavePlayer_4], properties {inputVolume [0.6,1.5]}}
 {id=QualityLevel_2, children [Enveloper_3,Enveloper_3,Enveloper_2,Enveloper_2]}
 {id=WavePlayer_2, properties {bLooping=true, soundWaveAssetPtr=/AmbientAudioContent/Waves/Biomes/Ambience_Crickets_Only_Loop_05_Quad}}
 {id=WavePlayer_3, properties {bLooping=true, soundWaveAssetPtr=/AmbientAudioContent/Waves/Biomes/Ambience_Crickets_Only_Loop_05_Stereo}}
 {
  id=WavePlayer_4
  properties {bLooping=true, soundWaveAssetPtr=/AmbientAudioContent/Waves/Biomes/Ambience_Biome_Chaparral_Morning_Quad_Loop_01}
 }
 {
  id=WavePlayer_5
  properties {bLooping=true, soundWaveAssetPtr=/AmbientAudioContent/Waves/Biomes/Ambience_Biome_Chaparral_Morning_Stereo_Loop_01}
 }
]
properties {
 attenuationSettings=/AmbientAudioBase/Attenuations/Fort_Ambient_Att
 bEnableBusSends=false
 maxDistance=2097152
 pitchMultiplier=0.9
 soundClassObject=&Fort_Ambient_Indoors_Athena
 volumeMultiplier=0.3
 concurrencySet [/AmbientAudioBase/Concurrency/AmbientConc_2Max_Loop_PreventNew_Default_Concurrency]
}
